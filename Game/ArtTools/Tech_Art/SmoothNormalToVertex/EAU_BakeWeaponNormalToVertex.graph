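# Editor utility Blueprint that bakes smoothed normals into the vertex colors of selected skeletal meshes, named for weapon meshes.
# TransferSmoothNormalToVertex runs UpdateMesh on each selection, which loops over LODs using two pooled dynamic meshes.
# One copy gets recomputed per-vertex normals, which CopyMeshData remaps from -1..1 to 0..1 RGB on the other.
# That mesh is written back into the original asset's LOD, with normals and tangents set to recompute.
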EditorUtilityBlueprint /Game/ArtTools/Tech_Art/SmoothNormalToVertex/EAU_BakeWeaponNormalToVertex
graphs [
 ConvertNormalToColor {nodes [
   n1 = ConvertNormalToColor!() -> (then, Normal_X:double, Normal_Y:double, Normal_Z:double) {is=title K2Node_FunctionEntry}
   > Return Node!(
    execute
    Color_R:float=float + float(A=float * float(A=@n1, B=0.5) {is=title K2Node_PromotableOperator}, B=0.5)
    Color_G:float=float + float(A=float * float(A=@n1.Normal_Y, B=0.5), B=0.5)
    Color_B:float=float + float(A=float * float(A=@n1.Normal_Z, B=0.5), B=0.5)
    Color_A:float=1
   )
  ]}
 CopyMeshData {nodes [
   n1 = CopyMeshData!() -> (then, TargetMesh:Dynamic Mesh Object Reference, OutputMesh:Dynamic Mesh Object Reference) {is=title K2Node_FunctionEntry}
   > n2 = GetMeshPerVertexNormals!(execute, TargetMesh:Dynamic Mesh Object Reference=@n1, bAverageSplitVertexValues:bool=true) -> (then, NormalList:Vector List Structure, bIsValidNormalSet:bool, bHasVertexIDGaps:bool, ReturnValue:Dynamic Mesh Object Reference)
   > n3 = ConvertVectorListToArray!(execute, VectorList:Vector List Structure=@n2) -> (then, VectorArray:Array of Vectors)
   > n4 = For Each Loop with Break!(Exec:Exec, Array:Array of Vectors=@n3, Break:Exec) -> (LoopBody:Exec, Array Element:Vector, Array Index:int, Completed:Exec)
   n5 = ConvertArrayToColorList!(execute=@n4.Completed, ColorArray:Array of Linear Color Structures=Get ColorList() -> (ColorList:Array of Linear Color Structures)) -> (then, ColorList:Color List Structure)
   > SetMeshPerVertexColors!(
    execute
    TargetMesh:Dynamic Mesh Object Reference=Get OutputMesh() -> (OutputMesh:Dynamic Mesh Object Reference)
    VertexColorList:Color List Structure=@n5
   ) -> (then, ReturnValue:Dynamic Mesh Object Reference)
   Add(
    execute=@n4.LoopBody
    TargetArray:Array of Linear Color Structures=Get ColorList()
    NewItem:"Linear Color Structure (by ref)"=ConvertNormalToColor!(Normal:Vector=@n4) -> (Color:Linear Color Structure) {is=title K2Node_CallFunction}
   ) {is=title K2Node_CallArrayFunction}
   Return Node!(execute)
  ]}
 EventGraph {nodes [MakeTransform(Rotation="0, 0, 0")]}
 TransferSmoothNormalToVertex {nodes [
   n1 = SpawnActorFromClass!(
    execute
    ActorClass:Actor Class Reference=&/GeometryScript_SmoothNormalToVertex.GeometryScript_SmoothNormalToVertex_C
    Location:Vector="0, 0, 0"
    Rotation:Rotator
    bTransient:bool
   ) -> (then, ReturnValue:Geometry Script Smooth Normal To Vertex Object Reference)
   > Set OriginalSkeletalMesh!(execute, OriginalSkeletalMesh:Skeletal Mesh Object Reference, self:Geometry Script Smooth Normal To Vertex Object Reference=@n1) -> (then, Output_Get:Skeletal Mesh Object Reference)
   > UpdateMesh(self:Geometry Script Smooth Normal To Vertex Object Reference=@n1)
   TransferSmoothNormalToVertex() {is=title K2Node_FunctionEntry}
   > n2 = GetSelectedAssetsOfClass(AssetClass=&SkeletalMesh)
   > n3 = For Each Loop(Array:Array of Object References=@n2) -> (Array Element:Object Reference)
   > n4 = Cast To SkeletalMesh!(execute, Object=@n3) -> (then, CastFailed, AsSkeletal Mesh:Skeletal Mesh Object Reference) {is=title K2Node_DynamicCast}
   > UpdateMesh!(execute, self:Self Object Reference, SkeletalMesh:Skeletal Mesh Object Reference=@n4) -> (then)
  ]}
 UpdateMesh {nodes [
   n1 = Get OriginalSkeletalMesh() -> (OriginalSkeletalMesh:Skeletal Mesh Object Reference)
   n2 = GetLODCount!(execute=@n10.then, SkeletalMesh:Skeletal Mesh Object Reference=Get OriginalSkeletalMesh()) -> (then, ReturnValue:int=0)
   > n3 = For Loop(LastIndex="Clamp (Integer)"(Value=@n2, Min=0, Max=int - int(A=@n2)) -> (ReturnValue=0))
   n4 = CopyMeshFromSkeletalMesh!(
    execute=@n15.then_0
    FromSkeletalMeshAsset:Skeletal Mesh Object Reference=@n1
    ToDynamicMesh:Dynamic Mesh Object Reference=Get TargetMesh() -> (TargetMesh:Dynamic Mesh Object Reference)
    AssetOptions_bApplyBuildSettings:bool=True
    AssetOptions_bRequestTangents:bool=True
    AssetOptions_bIgnoreRemoveDegenerates:bool=True
    AssetOptions_bUseBuildScale:bool=True
    RequestedLOD_LODType:EGeometryScriptLODType Enum=MaxAvailable
    RequestedLOD_LODIndex:int=@n3
   ) -> (Failure:Exec, Success:Exec, ReturnValue:Dynamic Mesh Object Reference)
   n5 = SetMeshConstantVertexColor!(
    execute
    TargetMesh:Dynamic Mesh Object Reference=@n4
    Color:Linear Color Structure=(R=1,G=1,B=1,A=1)
    Flags:Geometry Script Color Flags Structure
    bClearExisting:bool=true
   ) -> (then, ReturnValue:Dynamic Mesh Object Reference)
   ConvertMeshVertexColorsSRGBToLinear!(execute=[@n4.Success, @n5.then], TargetMesh:Dynamic Mesh Object Reference=@n4) -> (then, ReturnValue:Dynamic Mesh Object Reference)
   n6 = UpdateMesh!() -> (then, SkeletalMesh:Skeletal Mesh Object Reference) {is=title K2Node_FunctionEntry}
   > Set OriginalSkeletalMesh!(execute, OriginalSkeletalMesh=@n6) -> (then, Output_Get)
   > n7 = CreateDynamicMeshPool!(execute) -> (then, ReturnValue:Dynamic Mesh Pool Object Reference)
   > FreeAllMeshes(self:Dynamic Mesh Pool Object Reference=@n7)
   > n8 = RequestMesh!(execute, self:Dynamic Mesh Pool Object Reference=@n7) -> (then, ReturnValue:Dynamic Mesh Object Reference)
   > Set TargetMesh(TargetMesh:Dynamic Mesh Object Reference=@n8) -> (Output_Get:Dynamic Mesh Object Reference)
   > n9 = RequestMesh(self=@n7)
   > n10 = Set DyMesh(DyMesh:Dynamic Mesh Object Reference=@n9) -> (Output_Get:Dynamic Mesh Object Reference)
   n11 = ConvertMeshVertexColorsLinearToSRGB!(execute=@n15.then_2, TargetMesh:Dynamic Mesh Object Reference=Get TargetMesh()) -> (then, ReturnValue:Dynamic Mesh Object Reference)
   > CopyMeshToSkeletalMesh!(
    execute
    FromDynamicMesh:Dynamic Mesh Object Reference=@n11
    ToSkeletalMeshAsset:Skeletal Mesh Object Reference=Get OriginalSkeletalMesh()
    Options:Geometry Script Copy Mesh To Asset Options Structure=Make Geometry Script Copy Mesh to Asset Options!(
     bEnableRecomputeNormals:bool=true
     bEnableRecomputeTangents:bool=true
     bEnableRemoveDegenerates:bool=true
     NewMaterialUsage:ECopyMeshToAssetUseNewMaterialsOptions Enum=ReplaceAssetMaterials
     bCleanAssignedMaterials:bool=False
    ) -> (GeometryScriptCopyMeshToAssetOptions:Geometry Script Copy Mesh To Asset Options Structure) {is=title K2Node_MakeStruct}
    TargetLOD_bWriteHiResSource:bool=False
    TargetLOD_LODIndex:int=@n3
   ) -> (Failure:Exec, Success:Exec, ReturnValue:Dynamic Mesh Object Reference)
   n12 = CopyMeshFromSkeletalMesh!(
    execute=@n15.then_1
    FromSkeletalMeshAsset=@n1
    ToDynamicMesh=Get DyMesh() -> (DyMesh:Dynamic Mesh Object Reference)
    AssetOptions:Geometry Script Copy Mesh From Asset Options Structure
    RequestedLOD_LODType=MaxAvailable
    RequestedLOD_LODIndex=@n3
   ) -> (Failure, Success, ReturnValue)
   n13 = "Set Mesh To Per Vertex Normals (Computed)"!(execute=@n12.Success, TargetMesh:Dynamic Mesh Object Reference=@n12) -> (then, ReturnValue:Dynamic Mesh Object Reference) {is=title K2Node_CallFunction}
   > n14 = RecomputeNormals!(
    execute
    TargetMesh:Dynamic Mesh Object Reference=@n13
    CalculateOptions:Geometry Script Calculate Normals Options Structure
    bDeferChangeNotifications:bool=true
   ) -> (then, ReturnValue:Dynamic Mesh Object Reference)
   > CopyMeshData!(execute, self, TargetMesh:Dynamic Mesh Object Reference=@n14, OutputMesh:Dynamic Mesh Object Reference=Get TargetMesh()) -> (then) {is=title K2Node_CallFunction}
   n15 = Sequence!(execute=@n3.LoopBody) -> (then_0, then_1, then_2:Exec)
  ]}
]
variables [
 DyMesh {type=DynamicMesh, default=null}
 TargetMesh {type=DynamicMesh, default=null}
]
