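; Editor Geometry Script actor that bakes per-vertex normals into a skeletal mesh's vertex colors, then destroys itself.
; For each LOD, UpdateMesh copies the mesh into a white-colored target and a pooled scratch copy with recomputed normals.
; CopyMeshData takes vertex IDs from the scratch copy but reads normals from the target, remapping them by 0.5 scale and offset.
; The colors are converted linear-to-sRGB and written back to the original asset's LOD.
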(event Events|EventOnRebuildGeneratedMesh
  (params (TargetMesh :type Dynamic%20Mesh%20Object%20Reference))
  (Variables|Default|SetTargetMesh TargetMesh)
  (bind _returnvalue (GeometryScript|Utility|CreateDynamicMeshPool))
  (DynamicMesh|FreeAllMeshes _returnvalue)
  (bind _returnvalue_2 (DynamicMesh|RequestMesh _returnvalue))
  (Variables|Default|SetDyMesh _returnvalue_2))

(fn ConvertNormalToColor
  (params (Normal_X :type Float%20%28double-precision%29) (Normal_Y :type Float%20%28double-precision%29) (Normal_Z :type Float%20%28double-precision%29))
  (return (+ (* Normal_X 0.5) 0.5) (+ (* Normal_Y 0.5) 0.5) (+ (* Normal_Z 0.5) 0.5) 1.0))

(fn CopyMeshData
  (params (TargetMesh :type Dynamic%20Mesh%20Object%20Reference) (OutputMesh :type Dynamic%20Mesh%20Object%20Reference))
  ; Grab all vertices and prepare the color list to receive the data
  (bind _vertexidlist (GeometryScript|MeshQueries|GetAllVertexIDs TargetMesh))
  (LocalVariables|SetVertIDList _vertexidlist)
  (Utilities|Array|Resize (LocalVariables|GetColorList) (GeometryScript|ListUtils|GetIndexListLength (LocalVariables|GetVertIDList)))
  (bind _index (Utilities|FlowControl|ForLoopwithBreak 0 (GeometryScript|ListUtils|GetIndexListLastIndex (LocalVariables|GetVertIDList)))
    (:LoopBody
      ;[0.04844 0.06007 0.15 0.5] Convert and copy the normal value to a color list
      (bind _normallist (GeometryScript|Normals|GetMeshPerVertexNormals (Variables|Default|GetTargetMesh)))
      (Utilities|Array|SetArrayElem (LocalVariables|GetColorList) (.ReturnValue (GeometryScript|ListUtils|GetIndexListItem (LocalVariables|GetVertIDList) _index)) (CallFunction|ConvertNormaltoColor (.ReturnValue (GeometryScript|ListUtils|GetVectorListItem _normallist (.ReturnValue (GeometryScript|ListUtils|GetIndexListItem (LocalVariables|GetVertIDList) _index)))))))
    (:Completed
      ; Transfer the converted normal to color to the vertex color
      (bind _colorlist (GeometryScript|ListUtils|ConvertArraytoColorList (LocalVariables|GetColorList)))
      (GeometryScript|VertexColor|SetMeshPerVertexColors (Variables|GetOutputMesh) _colorlist))))

(fn UpdateMesh
  (bind _originalskeletalmesh (Variables|Default|GetOriginalSkeletalMesh))
  (Variables|Default|GetTargetMesh
    (:then
      _)
    (:else
      (bind _returnvalue (DynamicMeshComponent|GetDynamicMesh (Variables|DynamicMeshActor|GetDynamicMeshComponent)))
      (Variables|Default|SetTargetMesh _returnvalue)
      (bind _returnvalue_2 (GeometryScript|Utility|CreateDynamicMeshPool))
      (DynamicMesh|FreeAllMeshes _returnvalue_2)
      (bind _returnvalue_3 (DynamicMesh|RequestMesh _returnvalue_2))
      (Variables|Default|SetDyMesh _returnvalue_3)))
  (bind _returnvalue_4 (EditorScripting|SkeletalMesh|GetLODCount (Variables|Default|GetOriginalSkeletalMesh)))
  (for _index (range (+ (Math|Integer|Clamp(Integer) _returnvalue_4 0 (- _returnvalue_4 1)) 1))
    (Utilities|FlowControl|Sequence
      (:then_0
        (bind _returnvalue_5 (GeometryScript|SkeletalMesh|CopyMeshfromSkeletalMesh :FromSkeletalMeshAsset _originalskeletalmesh :ToDynamicMesh (Variables|Default|GetTargetMesh) :RequestedLOD (Utilities|Struct|MakeGeometryScriptMeshReadLOD "MaxAvailable" _index))
          (:Success
            (GeometryScript|VertexColor|SetMeshConstantVertexColor :TargetMesh _returnvalue_5 :Color "(R=1.000000,G=1.000000,B=1.000000,A=1.000000)" :bClearExisting true))))
      (:then_1
        (bind _returnvalue_6 (GeometryScript|SkeletalMesh|CopyMeshfromSkeletalMesh :FromSkeletalMeshAsset _originalskeletalmesh :ToDynamicMesh (Variables|Default|GetDyMesh) :RequestedLOD (Utilities|Struct|MakeGeometryScriptMeshReadLOD "MaxAvailable" _index))
          (:Success
            (bind _returnvalue_7 (GeometryScript|Normals|RecomputeNormals _returnvalue_6 (struct :bAngleWeighted true :bAreaWeighted true) true))
            (Default|CopyMeshData _returnvalue_7 (Variables|Default|GetTargetMesh)))))
      (:then_2
        (bind _returnvalue_8 (GeometryScript|VertexColor|ConvertMeshVertexColorsLineartoSRGB (Variables|Default|GetTargetMesh)))
        (GeometryScript|SkeletalMesh|CopyMeshtoSkeletalMesh :FromDynamicMesh _returnvalue_8 :ToSkeletalMeshAsset (Variables|Default|GetOriginalSkeletalMesh) :TargetLOD (Utilities|Struct|MakeGeometryScriptMeshWriteLOD false _index)))))
  (Actor|DestroyActor))

(fn ConstructionScript)
